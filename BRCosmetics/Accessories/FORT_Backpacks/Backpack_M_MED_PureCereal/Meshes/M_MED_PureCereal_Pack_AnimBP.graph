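# Animation Blueprint for the PureCereal backpack cosmetic, adding secondary motion to its flap, books, note, mask, knife and zipper.
# Character customization picks a pack-bone offset by body type and gender, overridden while skydiving.
# The offset and flap simulation are cached, then feed two differently tuned AnimDynamics chains, selected by bIsInFrontEnd.
# DynamicAlpha is also set from skydiving state but is not read in the graph.
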AnimBlueprint /BRCosmetics/Accessories/FORT_Backpacks/Backpack_M_MED_PureCereal/Meshes/M_MED_PureCereal_Pack_AnimBP
target_skeleton=&/M_MED_PureCereal_Pack_Skeleton
graphs [
 AnimGraph {nodes [
   n1 = Anim Dynamics - Bone: dyn_book_01(ComponentPose=Local To Component(LocalPose=Use cached pose 'PostTransform'()), Alpha=1, ExternalForce="0,0,0") {
    angularDampingOverride=0.9999
    angularSpringConstant=9000
    bAngularSpring=true
    bOverrideAngularDamping=true
    bOverrideLinearDamping=true
    bUseGravityOverride=true
    gravityScale=0.5
    linearDampingOverride=0.9999
    alphaBoolBlend {blendOption=Linear}
    boundBone {boneName=dyn_book_01}
    componentAppliedLinearAccClamp=<1000 1000 900>
    componentLinearAccScale=<0.3 0.3 0.25>
    componentLinearVelScale=<0.2 0.2 0.2>
    gravityOverride {z=-700}
    physicsBodyDefinitions [{
      boundBone {boneName=dyn_book_01}
      boxExtents=<25 20 10>
      constraintSetup {angularLimitsMax=<0 0 5>, angularLimitsMin=<0 0 -5>, angularTarget=<1 0 0>}
      localJointOffset=<0.3 0 0>
     }]
    simSpaceSettings {maxAngularAcceleration=100, maxAngularVelocity=100, simSpaceAngularAlpha=0.05, externalAngularVelocity {x=200}}
   }
   n2 = Anim Dynamics - Bone: dyn_book_02(ComponentPose=@n1, Alpha=1, ExternalForce="0,0,0") {
    angularDampingOverride=0.999
    bOverrideAngularDamping=true
    bOverrideLinearDamping=true
    bUseGravityOverride=true
    gravityScale=0.5
    linearDampingOverride=0.9999
    alphaBoolBlend {blendOption=Linear}
    boundBone {boneName=dyn_book_02}
    componentAppliedLinearAccClamp=<300 300 900>
    componentLinearAccScale=<0.1 0.1 0.2>
    componentLinearVelScale=<0.1 0.1 0.2>
    gravityOverride {z=-750}
    physicsBodyDefinitions [{
      boundBone {boneName=dyn_book_02}
      boxExtents=<25 25 25>
      constraintSetup {angularLimitsMax=<0 0 0>, angularLimitsMin=<0 0 -5>, angularTarget=<1 0 0>}
      localJointOffset=<0 0 -0.8>
     }]
    simSpaceSettings {maxAngularAcceleration=100, maxAngularVelocity=100, simSpaceAngularAlpha=0.1, externalAngularVelocity {x=150}}
   }
   n3 = Anim Dynamics - Bone: dyn_note_01(ComponentPose=@n2, Alpha=1, ExternalForce="0,0,0") {
    angularDampingOverride=0.999
    bOverrideAngularDamping=true
    bOverrideLinearDamping=true
    bUseGravityOverride=true
    gravityScale=0.5
    linearDampingOverride=0.9999
    alphaBoolBlend {blendOption=Linear}
    boundBone {boneName=dyn_note_01}
    componentAppliedLinearAccClamp=<300 300 900>
    componentLinearAccScale=<0.1 0.1 0.2>
    componentLinearVelScale=<0.1 0.1 0.2>
    gravityOverride {z=-750}
    physicsBodyDefinitions [{
      boundBone {boneName=dyn_note_01}
      boxExtents=<25 25 25>
      constraintSetup {angularLimitsMax=<0 0 0>, angularLimitsMin=<0 0 -5>, angularTarget=<1 0 0>}
      localJointOffset=<0 0 -0.8>
     }]
    simSpaceSettings {maxAngularAcceleration=100, maxAngularVelocity=100, simSpaceAngularAlpha=0.1, externalAngularVelocity {x=150}}
   }
   n4 = Anim Dynamics - Bone: dyn_mask(ComponentPose=@n3, Alpha=1, ExternalForce="0,0,0") {
    angularDampingOverride=0.9999
    angularSpringConstant=7000
    bAngularSpring=true
    bOverrideAngularDamping=true
    bOverrideLinearDamping=true
    bUseGravityOverride=true
    gravityScale=0.5
    linearDampingOverride=0.9999
    linearSpringConstant=80
    alphaBoolBlend {blendOption=Linear}
    boundBone {boneName=dyn_mask}
    componentAppliedLinearAccClamp=<1000 1000 550>
    componentLinearAccScale=<0.4 0.3 0.4>
    componentLinearVelScale=<0.3 0.4 0.3>
    gravityOverride {z=-700}
    physicsBodyDefinitions [{
      boundBone {boneName=dyn_mask}
      boxExtents=<20 20 25>
      constraintSetup {angularLimitsMax=<0 0 5>, angularLimitsMin=<0 0 -5>, angularTarget=<1 0 0>}
      localJointOffset=<0.5 0 0>
     }]
    simSpaceSettings {maxAngularAcceleration=100, maxAngularVelocity=100, simSpaceAngularAlpha=0.1}
   }
   n5 = Anim Dynamics - Bone: dyn_knife(ComponentPose=@n4, Alpha=1, ExternalForce="0,0,0") {
    angularDampingOverride=1
    bOverrideAngularDamping=true
    bUseGravityOverride=true
    gravityScale=0.5
    alphaBoolBlend {blendOption=Linear}
    boundBone {boneName=dyn_knife}
    componentAppliedLinearAccClamp=<600 600 550>
    componentLinearAccScale=<0.4 0.3 0.4>
    componentLinearVelScale=<0.3 0.4 0.3>
    gravityOverride {z=-400}
    physicsBodyDefinitions [{
      boundBone {boneName=dyn_knife}
      boxExtents=<25 25 25>
      constraintSetup {angularLimitsMax=<0 0 5>, angularLimitsMin=<0 0 0>, angularTarget=<0 0 0>}
      localJointOffset=<2 0 0>
     }]
    simSpaceSettings {maxAngularAcceleration=100, maxAngularVelocity=100, simSpaceAngularAlpha=0.1}
   }
   n6 = Anim Dynamics - Bone: dyn_zipper(ComponentPose=@n5, Alpha=1) {
    angularDampingOverride=0.8
    angularSpringConstant=100
    bAngularSpring=true
    bOverrideAngularDamping=true
    bUseGravityOverride=true
    bUsePlanarLimit=false
    alphaBoolBlend {blendOption=Linear}
    boundBone {boneName=dyn_zipper}
    componentAppliedLinearAccClamp=<800 800 2000>
    componentLinearAccScale=<0.2 0.2 0.35>
    componentLinearVelScale=<0.2 0.2 0.35>
    gravityOverride {z=-700}
    physicsBodyDefinitions [{
      boundBone {boneName=dyn_zipper}
      boxExtents=<15 12 17>
      constraintSetup {angularLimitsMax=<2.5 120 2.5>, angularLimitsMin=<-2.5 0 -2.5>, angularTarget=<1 0 0>}
      localJointOffset=<2 0 0>
     }]
    simSpaceSettings {maxAngularAcceleration=100, maxAngularVelocity=100, simSpaceAngularAlpha=0.1}
   }
   n7 = Anim Dynamics - Bone: dyn_book_01(ComponentPose=Local To Component(LocalPose=Use cached pose 'PostTransform'()), Alpha=1, ExternalForce="0,0,0") {
    angularDampingOverride=0.99
    bOverrideAngularDamping=true
    bOverrideLinearDamping=true
    bUseGravityOverride=true
    gravityScale=0.5
    linearDampingOverride=0.9999
    alphaBoolBlend {blendOption=Linear}
    boundBone {boneName=dyn_book_01}
    componentAppliedLinearAccClamp=<1000 1000 900>
    componentLinearAccScale=<0.3 0.3 0.25>
    componentLinearVelScale=<0.2 0.2 0.2>
    gravityOverride {z=-700}
    physicsBodyDefinitions [{
      boundBone {boneName=dyn_book_01}
      boxExtents=<25 25 25>
      constraintSetup {angularLimitsMax=<5 0 10>, angularLimitsMin=<-5 -8 -5>, angularTarget=<1 0 0>, linearAxesMax=<3 0 0>}
      localJointOffset=<1 0 0>
     }]
   }
   n8 = Anim Dynamics - Bone: dyn_book_02(ComponentPose=@n7, Alpha=1, ExternalForce="0,0,0") {
    angularDampingOverride=0.999
    angularSpringConstant=2000
    bAngularSpring=true
    bLinearSpring=true
    bOverrideAngularDamping=true
    bOverrideLinearDamping=true
    bUseGravityOverride=true
    gravityScale=0.5
    linearDampingOverride=0.9999
    linearSpringConstant=80
    alphaBoolBlend {blendOption=Linear}
    boundBone {boneName=dyn_book_02}
    componentAppliedLinearAccClamp=<300 300 900>
    componentLinearAccScale=<0.1 0.1 0.2>
    componentLinearVelScale=<0.1 0.1 0.2>
    gravityOverride {z=-750}
    physicsBodyDefinitions [{
      boundBone {boneName=dyn_book_02}
      boxExtents=<25 25 25>
      constraintSetup {angularLimitsMax=<5 0 0>, angularLimitsMin=<-5 -8 0>, angularTarget=<1 0 0>, linearAxesMax=<3 0 0>}
      localJointOffset=<0 0 -0.8>
     }]
   }
   n9 = Anim Dynamics - Bone: dyn_note_01(ComponentPose=@n8, Alpha=1, ExternalForce="0,0,0") {
    angularDampingOverride=0.999
    angularSpringConstant=2000
    bAngularSpring=true
    bLinearSpring=true
    bOverrideAngularDamping=true
    bOverrideLinearDamping=true
    bUseGravityOverride=true
    gravityScale=0.5
    linearDampingOverride=0.9999
    linearSpringConstant=80
    alphaBoolBlend {blendOption=Linear}
    boundBone {boneName=dyn_note_01}
    componentAppliedLinearAccClamp=<300 300 900>
    componentLinearAccScale=<0.1 0.1 0.2>
    componentLinearVelScale=<0.1 0.1 0.2>
    gravityOverride {z=-750}
    physicsBodyDefinitions [{
      boundBone {boneName=dyn_note_01}
      boxExtents=<25 25 25>
      constraintSetup {angularLimitsMax=<5 0 5>, angularLimitsMin=<-5 -8 -5>, angularTarget=<1 0 0>, linearAxesMax=<2 0 0>}
      localJointOffset=<0 0 -0.8>
     }]
   }
   n10 = Anim Dynamics - Bone: dyn_mask(ComponentPose=@n9, Alpha=1, ExternalForce="0,0,0") {
    angularDampingOverride=0.999
    angularSpringConstant=2000
    bAngularSpring=true
    bLinearSpring=true
    bOverrideAngularDamping=true
    bOverrideLinearDamping=true
    bUseGravityOverride=true
    gravityScale=0.5
    linearDampingOverride=0.9999
    linearSpringConstant=80
    alphaBoolBlend {blendOption=Linear}
    boundBone {boneName=dyn_mask}
    componentAppliedLinearAccClamp=<1000 1000 550>
    componentLinearAccScale=<0.4 0.3 0.4>
    componentLinearVelScale=<0.3 0.4 0.3>
    gravityOverride {z=-500}
    physicsBodyDefinitions [{
      boundBone {boneName=dyn_mask}
      boxExtents=<25 25 25>
      constraintSetup {angularLimitsMax=<0 10 10>, angularLimitsMin=<0 -10 -10>, angularTarget=<1 0 0>}
      localJointOffset=<1 0 0>
     }]
   }
   n11 = Anim Dynamics - Bone: dyn_knife(ComponentPose=@n10, Alpha=1, ExternalForce="0,0,0") {
    angularDampingOverride=1
    bOverrideAngularDamping=true
    bUseGravityOverride=true
    gravityScale=0.5
    alphaBoolBlend {blendOption=Linear}
    boundBone {boneName=dyn_knife}
    componentAppliedLinearAccClamp=<600 600 550>
    componentLinearAccScale=<0.4 0.3 0.4>
    componentLinearVelScale=<0.3 0.4 0.3>
    gravityOverride {z=-400}
    physicsBodyDefinitions [{
      boundBone {boneName=dyn_knife}
      boxExtents=<25 25 25>
      constraintSetup {angularLimitsMax=<5 5 5>, angularLimitsMin=<0 0 -10>, angularTarget=<0 0 0>}
      localJointOffset=<2 0 0>
     }]
   }
   n12 = Anim Dynamics - Bone: dyn_zipper(ComponentPose=@n11, Alpha=1) {
    angularDampingOverride=0.8
    angularSpringConstant=100
    bAngularSpring=true
    bOverrideAngularDamping=true
    bUseGravityOverride=true
    bUsePlanarLimit=false
    alphaBoolBlend {blendOption=Linear}
    boundBone {boneName=dyn_zipper}
    componentAppliedLinearAccClamp=<800 800 2000>
    componentLinearAccScale=<0.2 0.2 0.35>
    componentLinearVelScale=<0.2 0.2 0.35>
    gravityOverride {z=-700}
    physicsBodyDefinitions [{
      boundBone {boneName=dyn_zipper}
      boxExtents=<15 12 17>
      constraintSetup {angularLimitsMax=<2.5 120 2.5>, angularLimitsMin=<-2.5 0 -2.5>, angularTarget=<1 0 0>}
      localJointOffset=<2 0 0>
     }]
   }
   Output Pose(Result=Blend Poses by bool(BlendPose_0=Component To Local(ComponentPose=@n6), BlendPose_1=Component To Local(ComponentPose=@n12), bActiveValue=Get bIsInFrontEnd()) {blendType=Linear}) {layerGroup=None}
   n13 = "Transform (Modify) Bone - Bone: pack"(ComponentPose=Local To Component(LocalPose=Input Pose()), Translation=Get TargetPackOffset() -> (TargetPackOffset:Vector), Rotation="0,0,0") {translationMode=BMM_Additive, translationSpace=BCS_BoneSpace, alphaBoolBlend {blendOption=Linear}, boneToModify {boneName=pack}}
   n14 = Anim Dynamics - Bone: dyn_flap(ComponentPose=@n13, Alpha=1, ExternalForce="0,0,0") {
    angularDampingOverride=0.99
    bOverrideAngularDamping=true
    bUseGravityOverride=true
    alphaBoolBlend {blendOption=Linear}
    boundBone {boneName=dyn_flap}
    componentAppliedLinearAccClamp=<3000 3000 3000>
    componentLinearAccScale=<0.3 0.3 0.4>
    componentLinearVelScale=<0.3 0.3 0.4>
    gravityOverride {z=-980}
    physicsBodyDefinitions [{
      boundBone {boneName=dyn_flap}
      boxExtents=<15 10 10>
      constraintSetup {angularLimitsMax=<5 15 5>, angularLimitsMin=<-5 -40 -5>, angularTarget=<0 0 0>}
      localJointOffset=<4 0 0>
     }]
   }
   Save cached pose 'PostTransform'(Pose=Component To Local(ComponentPose=@n14)) {cacheName=PostTransform}
  ]}
 EventGraph {nodes [
   Event BlueprintUpdateAnimation()
   > Set DynamicAlpha(DynamicAlpha:float=SelectFloat(A=0.5, B=1, bPickA=Get bIsSkyDiving())) -> (Output_Get:float)
   > Set TargetPackOffset(TargetPackOffset:Vector=SelectVector(A="0,-2,0", B=Get CharacterPackOffset() -> (CharacterPackOffset:Vector), bPickA=Get bIsSkyDiving())) -> (Output_Get:Vector)
   n1 = Event OnCharacterCustomizationLoaded()
   n2 = Get Array Item with EFort Custom Part Type(Array:Array of Custom Character Part Object References=@n1.CustomParts) -> (Output:Custom Character Part Object Reference)
   n3 = "Equal (Enum)"(
    A:EFortCustomGender Enum=Get GenderPermitted!(self:Custom Character Part Object Reference=@n2) -> (GenderPermitted:EFortCustomGender Enum=Invalid)
    B:EFortCustomGender Enum=Female
   )
   n4 = "Equal (Enum)"(
    A:EFortCustomBodyType Enum=Get BodyTypesPermitted!(self:Custom Character Part Object Reference=@n2) -> (BodyTypesPermitted:EFortCustomBodyType Enum=NONE)
    B:EFortCustomBodyType Enum=Large
   )
   > Set CharacterPackOffset(CharacterPackOffset:Vector=SelectVector(A="0,-2,0", B=SelectVector(A="0,-0.7,0", B="0,-1.5,0", bPickA=@n3), bPickA=@n4)) -> (Output_Get:Vector)
   TryGetPawnOwner()
   OR Boolean(A=@n4, B=Get bIsSkyDiving())
  ]}
]
variables [
 CharacterPackOffset {type=Vector}
 DynamicAlpha {type=number, default=1}
 TargetPackOffset {type=Vector}
]
